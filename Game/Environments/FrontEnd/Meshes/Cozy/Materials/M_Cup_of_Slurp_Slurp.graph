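# Material for the Slurp liquid in the front-end cozy cup, fading from a deep blue emissive core to cyan Fresnel edges.
# A static switch, on by default, fakes liquid physics by rotating and shifting vertices by the angle between the object's up axis and world up.
# Height masks limit that offset and blend the normal map with a world-up normal, which also feeds the Fresnel, so the surface reads as level.
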Material /Game/Environments/FrontEnd/Meshes/Cozy/Materials/M_Cup_of_Slurp_Slurp
graph {
 properties {pixelDepthOffsetMode=PDOM_Legacy, usedShadingModels=Default Lit}
 nodes [
  Constant()
  n1 = Transform(None=Constant3Vector() {constant=<rgba 0 0 1 0>}) {transformSourceType=TRANSFORMSOURCE_Local}
  n2 = Constant3Vector() {constant=<rgba 0 0 1 0>}
  Divide(A=Tangent!(None=DotProduct(A=@n1, B=@n2))) {constB=360}
  n3 = Constant3Vector() {constant=<rgba 0 0 1 0>}
  n4 = TextureSampleParameter2D() {parameterName=Normal, samplerType=SAMPLERTYPE_Normal, texture=/Game/Environments/FrontEnd/Meshes/Cozy/Textures/T_Cup_of_Slurp_N}
  n5 = Transform(None=@n4.RGB)
  n6 = ComponentMask(None=Subtract(A=WorldPosition().XYZ, B=ObjectPositionWS())) {b=true}
  n7 = ScalarParameter() {defaultValue=-0.0818, parameterName=offset2}
  n8 = Saturate(None=Divide(A=Add(A=@n6, B=@n7), B=ScalarParameter() {defaultValue=-0.8023, parameterName=mask2}))
  n9 = LinearInterpolate(A=@n3, B=@n5, Alpha=@n8)
  n10 = Multiply(
   A=VectorParameter() {parameterName=Fresnel Edges Color, defaultValue=<rgba 0.2649 0.7377 0.8073 1>}.RGB
   B=Fresnel(
    ExponentIn=ScalarParameter() {defaultValue=2.5, parameterName=Fresnel_Exp}
    Normal=StaticSwitchParameter(True=@n9, False=@n5) {defaultValue=true, parameterName=UseFakeLiquidPhysics}
   )
  )
  n11 = ScalarParameter() {defaultValue=5, parameterName=Edges Color Multiply}
  n12 = VectorParameter() {parameterName=Core Color Emissive, defaultValue=<rgba 0 0.0403 0.1042 1>}
  n13 = ScalarParameter() {defaultValue=1, parameterName=Core Color Emissive Multiply}
  Reroute(None=Add(A=Multiply(A=@n10, B=@n11), B=Multiply(A=@n12.RGB, B=@n13)))
  ScalarParameter() {defaultValue=0.1, parameterName=Roughness}
  ScalarParameter() {defaultValue=1, parameterName=Specular}
  ScalarParameter() {parameterName=Metallic}
  n14 = Transform(None=Subtract(A=WorldPosition().XYZ, B=ObjectPositionWS())) {transformSourceType=TRANSFORMSOURCE_World, transformType=TRANSFORM_Local}
  n15 = Divide(
   A=Add(A=ComponentMask(None=@n14) {b=true}, B=ScalarParameter() {defaultValue=3.7683, parameterName=offset})
   B=ScalarParameter() {defaultValue=0.8023, parameterName=mask}
  )
  n16 = MaterialFunctionCall!(Normalized V1=Normalize(VectorInput=@n1), Normalized V2=Normalize(VectorInput=@n2)) {materialFunction=/Engine/Functions/Engine_MaterialFunctions02/Math/AngleBetweenVectors}
  n17 = Multiply(
   A=Multiply(A=Constant3Vector() {constant=<rgba 1 1 0 0>}, B=@n1)
   B=Multiply(A=ScalarParameter() {defaultValue=27.3259, parameterName=OffsetLiquid}, B=@n16.0-1)
  )
  n18 = RotateAboutAxis(
   NormalizedRotationAxis=Normalize(VectorInput=CrossProduct(A=@n1, B=@n2))
   RotationAngle=Multiply(A=ScalarParameter() {defaultValue=0.5, parameterName=Rot}, B=@n16.0-1)
   PivotPoint=TransformPosition(None=VectorParameter() {parameterName=Pivot, defaultValue=<rgba 0 0 7.6231 1>}.RGB) {transformType=TRANSFORMPOSSOURCE_World}
   Position=WorldPosition().XYZ
  )
  StaticSwitchParameter(True=Multiply(A=Saturate(None=@n15), B=Add(A=@n17, B=@n18)), False=Constant3Vector()) {defaultValue=true, parameterName=UseFakeLiquidPhysics}
  StaticSwitchParameter(
   True=LinearInterpolate(A=@n4.RGB, B=Transform(None=@n3) {transformSourceType=TRANSFORMSOURCE_World, transformType=TRANSFORM_Tangent}, Alpha=@n8)
   False=@n4.RGB
  ) {defaultValue=true, parameterName=UseFakeLiquidPhysics}
  VectorParameter() {parameterName=Base Color, defaultValue=<rgba 0 0.1268 0.3438 1>}
 ]
}
